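# Temporary blockout material for the HarvestHomes environment set in Asteria, using Default Lit shading and flagged for use with Nanite.
# The graph holds only three constant color vectors: black, a warm light grey and a neutral mid grey.
# It has no textures or parameters, and no links from the constants to material outputs are shown.
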Material /Game/Environments/Asteria/Sets/HarvestHomes/Materials/TEMP_HarvestHomes_Blockout
graph {
 properties {bUsedWithNanite=true, pixelDepthOffsetMode=PDOM_Legacy, usedShadingModels=Default Lit}
 nodes [
  Constant3Vector()
  Constant3Vector() {constant=<rgba 0.6042 0.5774 0.5507 0>}
  Constant3Vector() {constant=<rgba 0.3698 0.3698 0.3698 0>}
 ]
}
