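# TEMP_HarvestHomes_DarkMetal is a placeholder-named, Default Lit, Nanite-enabled material in the Asteria HarvestHomes environment set.
# Its graph holds only three constant grayscale color nodes: near-white, mid gray and dark gray.
# No textures or parameters are present, and no connections from the constants to material outputs are shown.
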Material /Game/Environments/Asteria/Sets/HarvestHomes/Materials/TEMP_HarvestHomes_DarkMetal
graph {
 properties {bUsedWithNanite=true, pixelDepthOffsetMode=PDOM_Legacy, usedShadingModels=Default Lit}
 nodes [
  Constant3Vector() {constant=<rgba 0.9062 0.9062 0.9062 0>}
  Constant3Vector() {constant=<rgba 0.1719 0.1719 0.1719 0>}
  Constant3Vector() {constant=<rgba 0.4948 0.4948 0.4948 0>}
 ]
}
