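# Temporary light metal material for the HarvestHomes environment set in Asteria, judging by its name and path.
# It uses Default Lit shading, is enabled for Nanite and contains only three greyscale constant colors, from white to mid grey.
# It has no textures or parameters, and the graph shows no links from the constants to material outputs.
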Material /Game/Environments/Asteria/Sets/HarvestHomes/Materials/TEMP_HarvestHomes_LightMetal
graph {
 properties {bUsedWithNanite=true, pixelDepthOffsetMode=PDOM_Legacy, usedShadingModels=Default Lit}
 nodes [
  Constant3Vector() {constant=<rgba 1 1 1 0>}
  Constant3Vector() {constant=<rgba 0.5625 0.5625 0.5625 0>}
  Constant3Vector() {constant=<rgba 0.2865 0.2865 0.2865 0>}
 ]
}
